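# Conversation database for the Lil Whip NPC, providing his greeting dialogue when a player starts talking to him.
# The entry tag leads to one speech node with a default album-plug line and contextual overrides.
# Overrides cover a first-ever conversation, prioritized by a curve row, and players wearing specific character skins via CID requirements.
# The speech node then links out to the shared NPC quest and basic services conversations.
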ConversationDatabase /NPCLibrary/NPCs/LilWhip/LilWhip/NPCLibrary_Conversation_LilWhip
entry_points [{tag=Conversation.Entry.LilWhip, destinations [#1]}]
nodes [
 ConversationEntryPointNode_7 {
  id=#1
  class=ConversationEntryPointNode
  outputs [#2]
  properties {entryTag {tagName=Conversation.Entry.LilWhip}}
 }
 Speech_C_4 {
  id=#2
  class=Speech_C
  outputs [#3, #4]
  properties {
   nodeName=""
   generalConfig {
    defaultMessage=PLEASE buy my album. I'll do anything. You need a cone guy? I know a cone guy...
    contextualMessages [
     {
      message="You're talking to the hardest soft-serve rapper in the game. "
      contextRequirements [{requirement {class=FortControllerRequirement_IsFirstEverConversationWithNPC}}]
      priority {curve {rowName=Tandem.Speech.Priority.FirstEverInteraction}}
     }
     {
      message=You're scheduled to watch the shop.
      contextRequirements [{requirement {
         class=FortControllerRequirement_HasCID
         properties {allowedCIds [/BRCosmetics/Athena/Items/Cosmetics/Characters/CID_460_Athena_Commando_F_BriteBomberSummer]}
        }}]
      priority {value=0, curve {curveTable=null, rowName=None}}
     }
     {
      message="Two kings, having a tête-à-tête. The rap gods would love this."
      contextRequirements [{requirement {
         class=FortControllerRequirement_HasCID
         properties {allowedCIds [/BRCosmetics/Athena/Items/Cosmetics/Characters/CID_A_424_Athena_Commando_M_Lancelot]}
        }}]
      priority {value=0, curve {curveTable=null, rowName=None}}
     }
     {
      message="Hey girl, are you a hired assassin or is my-- Oh, you are? Haha, we're cool right?"
      contextRequirements [{requirement {
         class=FortControllerRequirement_HasCID
         properties {allowedCIds [/BRCosmetics/Athena/Items/Cosmetics/Characters/CID_A_428_Athena_Commando_F_PinkWidow]}
        }}]
      priority {value=0, curve {curveTable=null, rowName=None}}
     }
     {
      message=Are you here to enact that justice you promised?
      contextRequirements [{requirement {
         class=FortControllerRequirement_HasCID
         properties {allowedCIds [/BRCosmetics/Athena/Items/Cosmetics/Characters/CID_A_427_Athena_Commando_F_Fuchsia]}
        }}]
      priority {value=0, curve {curveTable=null, rowName=None}}
     }
     {
      message="I cut your verse from \"The Armz Collector.\" I don't like being outshone RE: number of arms."
      contextRequirements [{requirement {
         class=FortControllerRequirement_HasCID
         properties {allowedCIds [/BRCosmetics/Athena/Items/Cosmetics/Characters/CID_A_429_Athena_Commando_M_Collectable]}
        }}]
      priority {value=0, curve {curveTable=null, rowName=None}}
     }
     {
      message="Does the Peace Syndicate offer protection? You know, from insurance-salesmen-slash-loan-sharks?"
      contextRequirements [{requirement {
         class=FortControllerRequirement_HasCID
         properties {allowedCIds [&CID_A_425_Athena_Commando_F_BlueJay]}
        }}]
      priority {value=0, curve {curveTable=null, rowName=None}}
     }
     {
      message=Can I sample that breathing sound for a track?
      contextRequirements [{requirement {
         class=FortControllerRequirement_HasCID
         properties {allowedCIds [&CID_A_422_Athena_Commando_M_Realm]}
        }}]
      priority {value=0, curve {curveTable=null, rowName=None}}
     }
    ]
   }
  }
 }
 ConversationLinkNode_0 {id=#4, class=ConversationLinkNode, properties {remoteEntryTag {tagName=Conversation.Entry.BasicServices}}}
 ConversationLinkNode_1 {id=#3, class=ConversationLinkNode, properties {remoteEntryTag {tagName=Conversation.Entry.NPCLibrary.Quests}}}
]
